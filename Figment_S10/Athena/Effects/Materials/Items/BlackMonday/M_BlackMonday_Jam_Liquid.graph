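# Default Lit material for Black Monday jam liquid, a scrolling, color-tinted surface with glowing bubbles.
# Three smoke-noise layers pan at different speeds and BumpOffset depths for fake parallax, then combine into a desaturated pattern.
# Bubbles come from offset noise-mask samples, warped by a panning swirl normal map and masked by scrolling smoke.
# Bubbles add to that pattern before tinting, and join a UV edge-fade mask to drive an emissive-strength color blend.
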Material /Figment_S10/Athena/Effects/Materials/Items/BlackMonday/M_BlackMonday_Jam_Liquid
graph {
 properties {
  bScreenSpaceReflections=true
  pixelDepthOffsetMode=PDOM_Legacy
  translucencyLightingMode=TLM_VolumetricDirectional
  usedShadingModels=Default Lit
 }
 nodes [
  Constant() {r=1}
  Constant() {r=0.1}
  n1 = TextureSample(UVs=Panner() {speedY=0.065}) {texture=/Game/Effects/Fort_Effects/Textures/Tile/T_Inky_Smoke_Tile}
  n2 = TextureCoordinate()
  n3 = AppendVector(A=ScalarParameter(), B=ScalarParameter() {defaultValue=0.25, parameterName=Param_1})
  n4 = Multiply(A=@n3) {constB=0.75}
  n5 = Add(A=Multiply(A=@n1.G) {constB=0.55}, B=BumpOffset(Coordinate=Panner(Coordinate=@n2, Speed=@n4), Height=Constant() {r=1}) {heightRatio=0.5})
  n6 = Multiply(A=@n3) {constB=0.55}
  n7 = Add(A=Multiply(A=@n1.G) {constB=0.225}, B=BumpOffset(Coordinate=Panner(Coordinate=@n2, Speed=@n6), Height=Constant() {r=1}) {heightRatio=-0.45})
  n8 = Multiply(A=@n1.G) {constB=0.1}
  n9 = BumpOffset(Coordinate=Panner(Coordinate=@n2, Speed=@n3), Height=Constant() {r=1}) {heightRatio=-1}
  n10 = AppendVector(
   A=AppendVector(
    A=TextureSample(UVs=@n5) {texture=&T_Inky_Smoke_Tile}.R
    B=TextureSample(UVs=@n7) {texture=&T_Inky_Smoke_Tile}.G
   )
   B=TextureSample(UVs=Add(A=@n8, B=@n9)) {texture=&T_Inky_Smoke_Tile}.B
  )
  n11 = TextureCoordinate()
  n12 = ScalarParameter() {defaultValue=0.2, parameterName=bubble Distortion UV Multipler}
  n13 = TextureSample(UVs=Panner(Coordinate=Multiply(A=@n11, B=@n12)) {speedY=0.125}) {samplerType=SAMPLERTYPE_Normal, texture=/Figment_S10/Environments/Foliage/Textures/T_SwirlNoise_N}
  n14 = AppendVector(
   A=Multiply(A=ComponentMask(None=@n13.RGB) {r=true}, B=ScalarParameter() {defaultValue=0.1, parameterName=Bubble Distortion Intensity U}) {constB=0.065}
   B=Multiply(A=ComponentMask(None=@n13.RGB) {g=true}, B=ScalarParameter() {defaultValue=0.15, parameterName=Bubble Distortion Intensity V}) {constB=0.125}
  )
  n15 = ScalarParameter() {defaultValue=0.35, parameterName=bubble Texture U Multipler}
  n16 = ScalarParameter() {defaultValue=0.6, parameterName=bubble Texture V Multipler}
  n17 = Panner(
   Coordinate=Multiply(A=TextureCoordinate() {uTiling=1.5, vTiling=1.5}, B=AppendVector(A=@n15, B=@n16))
   Speed=AppendVector(A=ScalarParameter() {parameterName=bubbles Speed U}, B=ScalarParameter() {defaultValue=0.15, parameterName=bubbles Speed V})
  ) {speedX=1, speedY=1}
  n18 = Constant2Vector() {g=0.01, r=0.01}
  n19 = Constant() {r=1}
  n20 = BumpOffset(Coordinate=Add(A=@n14, B=Add(A=@n17, B=@n18)), Height=@n19) {heightRatio=-0.525}
  n21 = Multiply(A=Subtract(A=TextureSample(UVs=@n20) {samplerType=SAMPLERTYPE_LinearColor, texture=/Game/Environments/Foliage/Textures/T_NoiseMask02_M}.G) {constB=0.55}) {constB=0.25}
  n22 = Add(A=@n14, B=Add(A=@n17, B=Constant2Vector() {g=-0.005}))
  n23 = OneMinus(None=TextureSample(UVs=BumpOffset(Coordinate=@n22, Height=@n19) {heightRatio=-0.6}) {samplerType=SAMPLERTYPE_LinearColor, texture=&T_NoiseMask02_M}.G)
  n24 = TextureSample(UVs=BumpOffset(Coordinate=Add(A=@n14, B=@n17), Height=@n19) {heightRatio=-0.6}) {samplerType=SAMPLERTYPE_LinearColor, texture=&T_NoiseMask02_M}
  n25 = Multiply(A=Multiply(A=Subtract(A=@n23) {constB=0.6}, B=Subtract(A=@n24.G) {constB=0.125})) {constB=12}
  n26 = Multiply(A=TextureCoordinate(), B=ScalarParameter() {defaultValue=0.35, parameterName=bubbome Mask UV Multiplier})
  n27 = TextureSample(UVs=Panner(Coordinate=@n26) {speedX=-0.04, speedY=0.065}) {texture=&T_Inky_Smoke_Tile}
  n28 = Multiply(A=Multiply(A=@n27.G, B=@n27.G), B=@n27.G)
  n29 = TextureSample(UVs=Panner(Coordinate=Multiply(A=@n26) {constB=1.5}) {speedX=0.04, speedY=0.2}) {texture=&T_Inky_Smoke_Tile}
  n30 = Multiply(A=Multiply(A=@n29.G, B=@n29.G), B=@n29.G)
  n31 = Add(A=Saturate(None=Multiply(A=@n28) {constB=15}), B=Saturate(None=Multiply(A=@n30) {constB=15}))
  n32 = Multiply(A=Add(A=Saturate(None=@n21), B=Power(Base=@n25) {constExponent=1.125}), B=Saturate(None=Power(Base=@n31)))
  Multiply(A=VectorParameter() {parameterName=Color, defaultValue=<rgba 0.6397 0 1 1>}.RGB, B=Add(A=Desaturation(None=@n10), B=@n32))
  n33 = ComponentMask(None=TextureCoordinate()) {r=true}
  n34 = Multiply(A=Multiply(A=OneMinus(None=@n33), B=@n33), B=ScalarParameter() {defaultValue=3, parameterName=WindowEdgesFadeMultiplier})
  n35 = ScalarParameter() {defaultValue=1, parameterName=WindowEdgesFadePower}
  n36 = Constant()
  n37 = ScalarParameter() {defaultValue=0.05, parameterName=OffsetY}
  n38 = ComponentMask(None=Add(A=TextureCoordinate(), B=AppendVector(A=@n36, B=@n37))) {g=true}
  n39 = OneMinus(None=@n38)
  n40 = Multiply(
   A=Power(Base=Multiply(A=@n39, B=@n38), Exp=ScalarParameter() {defaultValue=5, parameterName=WindowEdgesFadePowerVert})
   B=ScalarParameter() {defaultValue=1250, parameterName=WindowEdgesFadeMultiplierVert}
  )
  n41 = Saturate(None=Multiply(A=Power(Base=@n34, Exp=@n35), B=Saturate(None=@n40)))
  n42 = Multiply(A=@n32, B=ScalarParameter() {defaultValue=6, parameterName=Bubbles Emissive Multiplier})
  Multiply(
   A=LinearInterpolate(B=VectorParameter() {parameterName=Color, defaultValue=<rgba 0.6397 0 1 1>}.RGB, Alpha=Add(A=@n41, B=@n42))
   B=ScalarParameter() {defaultValue=0.25, parameterName=EmissiveMult}
  )
 ]
}
