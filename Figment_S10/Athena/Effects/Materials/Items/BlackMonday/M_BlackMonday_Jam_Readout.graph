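# Material for the Black Monday "Jam" item's readout screen, drawing a scrolling LED-style display with glowing edges.
# A stepping panner time jumps the scaled readout texture horizontally, and its brightened color is masked over a dark background by a tiled, sharpened LED dot texture.
# A cubed flux mask plus a tiled analog scanline texture then blends that result toward an edge glow color.
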Material /Figment_S10/Athena/Effects/Materials/Items/BlackMonday/M_BlackMonday_Jam_Readout
graph {
 properties {pixelDepthOffsetMode=PDOM_Legacy, usedShadingModels=Default Lit}
 nodes [
  Constant() {r=1}
  Constant() {r=0.1}
  n1 = Multiply(A=TextureCoordinate(), B=Constant2Vector() {g=0.825, r=0.85})
  n2 = Constant2Vector() {g=0.11}
  n3 = MaterialFunctionCall!(
   TimeSpeed=ScalarParameter() {defaultValue=0.2, parameterName=Speed}
   Jump Amount Per Step=ScalarParameter() {defaultValue=30, parameterName=Step Size}
   Number of Steps Per Unit of Time=ScalarParameter() {defaultValue=30, parameterName=Num Steps}
   Time Offset
  ) {materialFunction=/Engine/Functions/Engine_MaterialFunctions02/SteppingPannerTime}
  n4 = Constant()
  n5 = TextureSample(UVs=Add(A=Add(A=@n1, B=@n2), B=AppendVector(A=@n3.Result, B=@n4))) {texture=/Figment_S10/Athena/Effects/Textures/Items/BlackMonday/T_BlackMonday_Readout_D}
  n6 = ScalarParameter() {defaultValue=8, parameterName=Emissive Mult}
  n7 = TextureSample!(
   UVs=Multiply(A=TextureCoordinate() {uTiling=2}, B=ScalarParameter() {defaultValue=24, parameterName=LEDScreenTextureTile})
   Tex
   Bias=Constant() {r=-1}
   Apply View MipBias
  ) {mipValueMode=TMVM_MipBias, texture=/Game/Effects/Fort_Effects/Textures/Plasma/T_PlasmaUV_Mask}
  n8 = TextureCoordinate()
  n9 = VectorParameter() {parameterName=34234, defaultValue=<rgba 0.75 1.5 0.125 -0.225>}
  n10 = AppendVector(A=@n9.RGB, B=@n9.RGB)
  n11 = ComponentMask(None=@n10) {g=true, r=true}
  n12 = TextureSample!(UVs=Add(A=Multiply(A=@n8, B=@n11), B=ComponentMask(None=@n10) {a=true, b=true}), Tex, Apply View MipBias) {texture=/Figment_S10/Missions/Primary/OB_FarmsteadFort/ArtAssets/Textures/T_FluxMask}
  n13 = Multiply(A=@n12.G, B=@n12.G)
  n14 = TextureCoordinate() {vTiling=64}
  LinearInterpolate(
   A=LinearInterpolate(
    A=VectorParameter() {parameterName=BG_Color, defaultValue=<rgba 0.0001 0.0016 0.0052 1>}.RGB
    B=Multiply(A=@n5.RGB, B=@n6)
    Alpha=Power(Base=@n7.G) {constExponent=5}
   )
   B=VectorParameter() {parameterName=EdgeGlowColor, defaultValue=<rgba 0.2604 0 0.174 0>}.RGB
   Alpha=Add(A=Multiply(A=@n13, B=@n12.G), B=TextureSample(UVs=@n14) {texture=/Game/Effects/Fort_Effects/Textures/Tile/T_ScanLine_Analog}.G)
  )
 ]
}
