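# Additive, unlit, two-sided hologram-style material with depth test disabled, apparently for the Black Monday item's vision effect on particles.
# Particle color is brightened by a panning screen-space scanline texture, and particle alpha is modulated by hard stripes along world Z.
# Vertex red blends two time-driven rotations about the local X axis at the particle position, spinning rings at separate speeds.
# Dynamic parameters blend in and bias a hologram wiggle function added to that offset.
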Material /Figment_S10/Athena/Effects/Materials/Items/BlackMonday/M_Black_Monday_Vision
graph {
 properties {
  bDisableDepthTest=true
  bEnableResponsiveAA=true
  bUsedWithClothing=true
  bUsedWithMeshDeformer=true
  bUsedWithMeshParticles=true
  bUsedWithMorphTargets=true
  bUsedWithParticleSprites=true
  bUsedWithSkeletalMesh=true
  blendMode=BLEND_Additive
  pixelDepthOffsetMode=PDOM_Legacy
  shadingModel=MSM_Unlit
  twoSided=true
  usedShadingModels=Unlit
 }
 nodes [
  n1 = ParticleColor()
  n2 = Panner(Coordinate=Multiply(A=ScreenPosition().ViewportUV) {constB=4}, Time=Time()) {speedX=0.1, speedY=0.25}
  Add(
   A=@n1.RGB
   B=Multiply(A=Multiply(A=@n1.RGB) {constB=0.5}, B=TextureSample(UVs=@n2) {texture=/Game/Effects/Fort_Effects/Textures/Tile/T_ScanLine_Analog}.R)
  )
  Constant3Vector() {constant=<rgba 0.0937 0.1719 0.25 1>}
  Constant()
  n3 = Transform(None=Constant3Vector() {constant=<rgba 1 0 0 0>}) {transformSourceType=TRANSFORMSOURCE_Local}
  n4 = Frac(None=Multiply(A=Time(), B=ScalarParameter() {defaultValue=-0.05, parameterName=RotationSpeed_Ring_01}))
  n5 = ParticlePositionWS()
  n6 = WorldPosition()
  n7 = Frac(None=Multiply(A=Time(), B=ScalarParameter() {defaultValue=0.1, parameterName=RotationSpeed_Ring_02}))
  n8 = LinearInterpolate(
   A=RotateAboutAxis(NormalizedRotationAxis=@n3, RotationAngle=ComponentMask(None=@n4) {r=true}, PivotPoint=@n5, Position=@n6.XYZ)
   B=RotateAboutAxis(NormalizedRotationAxis=@n3, RotationAngle=ComponentMask(None=@n7) {r=true}, PivotPoint=@n5, Position=@n6.XYZ)
   Alpha=VertexColor().R
  )
  n9 = VectorParameter() {parameterName=WPO Intensity, defaultValue=<rgba 0 80 0 1>}
  n10 = DynamicParameter() {defaultValue=<rgba 1 0 0 0>, paramNames [Wiggle_Activate, 0, 1, TriggerWiggle]}
  LinearInterpolate(
   A=@n8
   B=Add(
    A=MaterialFunctionCall!(WPO Intensity A=@n9.RGB, WPO Intensity B=@n9.RGB, Time, Bias=@n10.TriggerWiggle) {materialFunction=/Game/Effects/Fort_Effects/Effects/Atlas/Materials/MF_Atlas_HologramWiggle_01}.Result
    B=@n8
   )
   Alpha=@n10.Wiggle_Activate
  )
  n11 = Multiply(A=ComponentMask(None=WorldPosition().XYZ) {b=true}) {constB=0.4}
  n12 = Floor(None=Add(A=Frac(None=@n11)) {constB=0.2})
  Multiply(A=ParticleColor().A, B=Clamp(None=Add(A=@n12) {constB=0.325}))
 ]
}
